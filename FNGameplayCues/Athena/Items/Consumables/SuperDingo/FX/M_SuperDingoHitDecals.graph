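# Translucent unlit Niagara sprite material for SuperDingo hit decals, a particle-colored glow visible near intersecting scene geometry.
# Fresnel on flattened SceneTexture world normals is multiplied by explosion normal-map noise, with a constant replacing Fresnel on mobile.
# An inverted sphere mask around HitLocation, sized by dynamic-parameter GlowPan, clears the center.
# The result feeds HDR emissive and, with particle alpha, radial UV falloff and inverted depth fade, opacity.
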Material /FNGameplayCues/Athena/Items/Consumables/SuperDingo/FX/M_SuperDingoHitDecals
graph {
 properties {bUsedWithNiagaraSprites=true, blendMode=BLEND_Translucent, pixelDepthOffsetMode=PDOM_Legacy, shadingModel=MSM_Unlit, usedShadingModels=Unlit}
 nodes [
  n1 = SceneTexture() {sceneTextureId=PPI_WorldNormal}
  n2 = Fresnel(
   ExponentIn=DynamicParameter() {paramNames [FresnelExponent, Refraction, GlowPan, Param4]}.FresnelExponent
   Normal=MaterialFunctionCall(Normal=ComponentMask(None=@n1.Color) {b=true, g=true, r=true}, Flatness=Constant() {r=0.2}) {materialFunction=&FlattenNormal}.Result
  ) {baseReflectFraction=0, exponent=1}
  n3 = ShadingPathSwitch(Default=@n2, Deferred=@n2, Forward=Constant3Vector() {constant=<rgba 0 0 1 0>}, Mobile=@n2)
  n4 = Multiply(
   A=TextureCoordinate()
   B=AppendVector(
    A=ScalarParameter() {defaultValue=0.6, group=Dissolve, parameterName=Dissolve U HQ}
    B=ScalarParameter() {defaultValue=0.6, group=Dissolve, parameterName=Dissolve V HQ}
   )
  )
  n5 = VertexInterpolator(VS=PreSkinnedPosition())
  n6 = ScalarParameter() {defaultValue=150, parameterName=Object Radius}
  n7 = DynamicParameter() {paramNames [FresnelExponent, Refraction, GlowPan, Param4]}
  n8 = OneMinus(None=SphereMask(
    A=TransformPosition(None=@n5.PS) {transformType=TRANSFORMPOSSOURCE_World}
    B=VectorParameter() {parameterName=HitLocation, defaultValue=<rgba 0 0 80 0>}.RGB
    Radius=Multiply(A=@n6, B=@n7.GlowPan)
    Hardness=Constant() {r=0.5}
   ) {hardnessPercent=50})
  n9 = Multiply(
   A=ConstantBiasScale(None=TextureSample(UVs=@n4) {samplerType=SAMPLERTYPE_Normal, texture=/Game/Effects/Fort_Effects/Gameplay/SmokeDustExplosions/Textures/T_Explosion_NM_01}.R)
   B=@n8
  )
  n10 = ParticleColor()
  n11 = Multiply(
   A=ShadingPathSwitch(Default=@n3, Deferred=@n3, Mobile=Constant() {r=0.3}) {desc=Migrated from FeatureLevelSwitch}
   B=Multiply(A=LinearInterpolate(B=@n9, Alpha=@n8), B=@n10.RGB)
  )
  MaterialFunctionCall(Emissive=MaterialFunctionCall(LDR Input=Clamp(None=@n11), HDR Tint=@n11) {materialFunction=&DeriveHDRfromLDR}.RemappedRange) {materialFunction=&MF_FX_EmissiveAdjustments}
  n12 = Constant() {r=0.5}
  n13 = SphereMask(A=TextureCoordinate(), B=@n12, Radius=@n12) {hardnessPercent=0}
  n14 = OneMinus(None=DepthFade(FadeDistance=ScalarParameter() {defaultValue=100, parameterName=VisibleDepthDistance}) {fadeDistanceDefault=80})
  n15 = ScalarParameter() {defaultValue=15, parameterName=DepthFade}
  Multiply(A=Saturate(None=@n11), B=Multiply(A=Multiply(A=@n10.A, B=@n13), B=DepthFade(Opacity=@n14, FadeDistance=@n15) {fadeDistanceDefault=15}))
 ]
}
